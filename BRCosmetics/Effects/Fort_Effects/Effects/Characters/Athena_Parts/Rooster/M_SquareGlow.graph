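# M_SquareGlow is an unlit, additive, two-sided glow material for Niagara mesh particles in the Rooster character effects.
# It squares an inverted V-coordinate gradient, tints it by particle color and alpha, and feeds it to MF_FX_EmissiveAdjustments as emissive.
# A separate term multiplies a parameterized DepthFade by inverted Fresnel, a soft falloff at intersections and grazing angles.
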Material /BRCosmetics/Effects/Fort_Effects/Effects/Characters/Athena_Parts/Rooster/M_SquareGlow
graph {
 properties {
  bUsedWithNiagaraMeshParticles=true
  blendMode=BLEND_Additive
  pixelDepthOffsetMode=PDOM_Legacy
  shadingModel=MSM_Unlit
  twoSided=true
  usedShadingModels=Unlit
 }
 nodes [
  n1 = OneMinus(None=ComponentMask(None=TextureCoordinate()) {g=true})
  n2 = Multiply(A=@n1, B=@n1)
  n3 = ParticleColor()
  MaterialFunctionCall(Emissive=Multiply(A=Multiply(A=@n2, B=@n3.RGB), B=@n3.A)) {materialFunction=&MF_FX_EmissiveAdjustments}
  Multiply(A=DepthFade(FadeDistance=ScalarParameter() {defaultValue=5, parameterName=FadeDistance}), B=OneMinus(None=Fresnel()))
 ]
}
