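# Niagara system for the Rooster character cosmetic's helmet, providing glow and light effects on the head.
# Five emitters cover left and right eye lights, a forehead glow, a forehead glow mesh and a top light.
# A User.bisHeadHidden bool is exposed, and bool-to-float and bool-to-vector dynamic inputs suggest it switches effect values when the head is hidden.
# A Set Variables module runs at system spawn alongside SystemState.
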NiagaraSystem /BRCosmetics/Effects/Fort_Effects/Effects/Characters/Athena_Parts/Rooster/NS_Rooster_Helmet
dependencies {
 usedDataInterfaces []
 usedDynamicInputs [
  &MakeCustomVector2DFromBool
  &Vector2DFromFloat
  &MakeCustomVectorFromBool
  &MakeCustomFloatFromBool
 ]
 usedModules [
  &SystemState
  &self:SystemSpawnScript.SystemScriptSource.SystemScriptGraph.NiagaraNodeAssignment_0.SetVariables_#1
 ]
}
emitters {
 EyeLightL {rendererClasses [], renderers []}
 EyeLightR {rendererClasses [], renderers []}
 ForeheadGlow {rendererClasses [], renderers []}
 Forehead_Glow_Mesh {rendererClasses [], renderers []}
 Top_Light {rendererClasses [], renderers []}
}
summary {
 systemName=NS_Rooster_Helmet
 emitters [
  {emitterName=ForeheadGlow, rendererClasses []}
  {emitterName=Forehead_Glow_Mesh, rendererClasses []}
  {emitterName=Top_Light, rendererClasses []}
  {emitterName=EyeLightR, rendererClasses []}
  {emitterName=EyeLightL, rendererClasses []}
 ]
 userVariables [User.bisHeadHidden {type=NiagaraBool}]
}
